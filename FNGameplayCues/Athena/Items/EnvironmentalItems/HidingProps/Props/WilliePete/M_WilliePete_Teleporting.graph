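# Translucent unlit mesh-particle material for the WilliePete hiding prop's teleport effect, drawn without depth test as a swirling water tunnel.
# Panning, V-warped UV layers build water, caustics, speed lines and mist, with normal-driven cubemap reflection and fake lighting.
# Distorted foam is remapped through curve-atlas rows for color and optional frame opacity.
# Particle dynamic parameters drive whitewash, emissive and foam; a time-driven circular wobble with distance falloff provides vertex offset.
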Material /FNGameplayCues/Athena/Items/EnvironmentalItems/HidingProps/Props/WilliePete/M_WilliePete_Teleporting
graph {
 properties {
  bDisableDepthTest=true
  bUseTranslucencyVertexFog=false
  bUsedWithMeshParticles=true
  blendMode=BLEND_Translucent
  pixelDepthOffsetMode=PDOM_Legacy
  shadingModel=MSM_Unlit
  usedShadingModels=Unlit
 }
 nodes [
  n1 = TextureCoordinate()
  n2 = Multiply(
   A=Power(Base=ComponentMask(None=@n1) {g=true}, Exp=ScalarParameter() {defaultValue=1, group=Foam, parameterName=Foam_Whirl_Power, sortPriority=1}) {constExponent=1}
   B=ScalarParameter() {defaultValue=3, group=Foam, parameterName=Foam_Whirl_Intensity, sortPriority=2}
  )
  n3 = Constant()
  n4 = VectorParameter() {group=Foam, parameterName=Foam_UV_Ctrl, defaultValue=<rgba 3 2 0 -1.5>}
  n5 = AppendVector(A=@n4.R, B=@n4.R)
  n6 = Panner(Coordinate=Add(A=AppendVector(A=@n2, B=@n3), B=Multiply(A=@n1, B=@n5)), Speed=AppendVector(A=@n4.B, B=@n4.B))
  n7 = TextureCoordinate()
  n8 = ComponentMask(None=@n7) {g=true}
  n9 = ScalarParameter() {defaultValue=1, group=Foam, parameterName=FoamMaskDistort_Whirl_Power, sortPriority=1}
  n10 = AppendVector(
   A=Multiply(
    A=Power(Base=@n8, Exp=@n9) {constExponent=1}
    B=ScalarParameter() {defaultValue=3, group=Foam, parameterName=FoamMaskDistort_Whirl_Intensity, sortPriority=2}
   )
   B=Constant()
  )
  n11 = VectorParameter() {group=Foam, parameterName=FoamMaskDistort_UV_Ctrl, defaultValue=<rgba 3 2 0 -0.5>}
  n12 = Multiply(A=@n7, B=AppendVector(A=@n11.R, B=@n11.R))
  n13 = TextureSampleParameter2D(UVs=Panner(Coordinate=Add(A=@n10, B=@n12), Speed=AppendVector(A=@n11.B, B=@n11.B))) {
   group=Foam
   parameterName=FoamDistorsion
   samplerType=SAMPLERTYPE_Normal
   texture=/Game/Packages/DS_Fortnite_TRV_Terrain/Textures/T_TRV_SwampWater_01_N
  }
  n14 = Multiply(A=@n13.R) {constB=0.1}
  n15 = TextureSampleParameter2D(UVs=Add(A=@n6, B=@n14)) {
   group=Foam
   parameterName=Foam
   samplerType=SAMPLERTYPE_LinearColor
   texture=/Game/Athena/Apollo/Water/Textures/02_Rounded/03_WaterFoam_Veroni_base_H
  }
  n16 = Multiply(A=@n15.R) {constB=0.5}
  n17 = TextureSampleParameter2D(UVs=Add(A=@n14, B=Multiply(A=@n6) {constB=3})) {parameterName=Param, samplerType=SAMPLERTYPE_Normal, texture=/Game/Athena/Apollo/Water/Textures/02_Rounded/03_WaterFoam_Veroni_base_N}
  Add(A=@n16, B=Multiply(A=@n17.R) {constB=0.5})
  n18 = MaterialFunctionCall() {materialFunction=&LinearGradient}
  n19 = Power(Base=@n18.VGradient, Exp=ScalarParameter() {defaultValue=4, parameterName=GradientPowerInnerShadow})
  n20 = OneMinus(None=@n19)
  n21 = DynamicParameter() {defaultValue=<rgba 0 1 0 0>, paramNames [WhiteWash, Final Emissive Multiplier, Foam_Threshold, FoamFrame_Offset]}
  n22 = LinearInterpolate(A=Constant() {r=-1}, B=Constant() {r=1}, Alpha=@n21.WhiteWash)
  Clamp(None=Multiply(A=Add(A=@n20, B=@n22), B=VectorParameter() {parameterName=TunnelExitColor, defaultValue=<rgba 1 1 1 0>}.RGB))
  n23 = Multiply(A=Time(), B=ScalarParameter() {defaultValue=0.25, parameterName=RotationSpeed})
  n24 = Add(A=@n23, B=Sine(None=Multiply(A=@n23) {constB=0.45}))
  n25 = MaterialFunctionCall(R=Sine(None=@n24), B=Cosine(None=@n24)) {materialFunction=&AppendMany}
  n26 = ScalarParameter() {defaultValue=150, parameterName=WPO_Multiplier}
  n27 = WorldPosition() {worldPositionShaderOffset=WPT_ExcludeAllShaderOffsets}
  n28 = TransformPosition(None=VectorParameter() {parameterName=Offset, defaultValue=<rgba 0 50 0 1>}.RGB) {transformType=TRANSFORMPOSSOURCE_World}
  n29 = Saturate(None=Divide(A=Distance(A=@n27.XYZ, B=@n28), B=ScalarParameter() {defaultValue=150, parameterName=Rad}))
  LinearInterpolate(
   A=Constant3Vector()
   B=Transform(None=Multiply(A=@n25.RGB, B=@n26)) {transformSourceType=TRANSFORMSOURCE_Local}
   Alpha=Power(Base=OneMinus(None=@n29), Exp=ScalarParameter() {defaultValue=2, parameterName=LinearFalloffPower})
  )
  n30 = TextureCoordinate()
  n31 = ComponentMask(None=@n30) {g=true}
  n32 = ScalarParameter() {defaultValue=1, group=AirFlow, parameterName=AF_Whirl_Power, sortPriority=1}
  n33 = AppendVector(
   A=Multiply(
    A=Power(Base=@n31, Exp=@n32) {constExponent=1}
    B=ScalarParameter() {defaultValue=1, group=AirFlow, parameterName=AF_Whirl_Intensity, sortPriority=2}
   )
   B=Constant()
  )
  n34 = VectorParameter() {group=AirFlow, parameterName=AirFlow_UV_Ctrl, defaultValue=<rgba 3 0.3 0.2 0.5>}
  n35 = Multiply(A=@n30, B=AppendVector(A=@n34.R, B=@n34.R))
  n36 = TextureSample(UVs=Panner(Coordinate=Add(A=@n33, B=@n35), Speed=AppendVector(A=@n34.B, B=@n34.B))) {
   samplerSource=SSM_Wrap_WorldGroupSettings
   texture=/Game/CharClasses/Outlander/Gadgets/Fragment/ShockTower/Textures/T_ShockTower_LightningOffset_01
  }
  n37 = LinearInterpolate(
   A=VectorParameter() {group=Emissive, parameterName=Emissive Light Color, defaultValue=<rgba 1 1 1 0>}.RGB
   B=VectorParameter() {group=Emissive, parameterName=Emissive Light Color 2, defaultValue=<rgba 1 1 1 0>}.RGB
   Alpha=@n36.B
  )
  n38 = TextureCoordinate()
  n39 = Multiply(
   A=Power(
    Base=ComponentMask(None=@n38) {g=true}
    Exp=ScalarParameter() {defaultValue=1, group=BaseTexture, parameterName=BT_Whirl_Power, sortPriority=1}
   ) {constExponent=1}
   B=ScalarParameter() {defaultValue=1, group=BaseTexture, parameterName=BT_Whirl_Intensity, sortPriority=2}
  )
  n40 = Constant()
  n41 = VectorParameter() {group=BaseTexture, parameterName=BaseTexture_UV_Ctrl, defaultValue=<rgba 2 0.3 0.2 -0.5>}
  n42 = AppendVector(A=@n41.R, B=@n41.R)
  n43 = Panner(Coordinate=Add(A=AppendVector(A=@n39, B=@n40), B=Multiply(A=@n38, B=@n42)), Speed=AppendVector(A=@n41.B, B=@n41.B))
  n44 = TextureSample(UVs=@n43) {samplerType=SAMPLERTYPE_LinearColor, texture=/Game/Global/Textures/Default/Tileable_Patterns/Cloud_Diffuse}
  n45 = TextureSample(UVs=@n43) {texture=/Game/Packages/Fortress_Transylvania/Haunted_Forest/Swamps/Textures/2ndInfectedMapLinear2}
  n46 = TextureCoordinate()
  n47 = ComponentMask(None=@n46) {g=true}
  n48 = ScalarParameter() {defaultValue=1, group=Caustic, parameterName=Caustic_Whirl_Power, sortPriority=1}
  n49 = AppendVector(
   A=Multiply(
    A=Power(Base=@n47, Exp=@n48) {constExponent=1}
    B=ScalarParameter() {defaultValue=1, group=Caustic, parameterName=Caustic_Whirl_Intensity, sortPriority=2}
   )
   B=Constant()
  )
  n50 = VectorParameter() {group=Caustic, parameterName=Caustic_UV_Ctrl, defaultValue=<rgba 2 0.3 0.2 -0.5>}
  n51 = Multiply(A=@n46, B=AppendVector(A=@n50.R, B=@n50.R))
  n52 = TextureSample(UVs=Panner(Coordinate=Add(A=@n49, B=@n51), Speed=AppendVector(A=@n50.B, B=@n50.B))) {samplerType=SAMPLERTYPE_LinearColor, texture=/Game/Global/Textures/Default/Caustics/03_WaterFoam_AccentSwirls_H}
  n53 = ScalarParameter() {defaultValue=1, group=Caustic, parameterName=Caustic_Intensity_Power, sortPriority=3}
  n54 = Add(
   A=MaterialFunctionCall(
    A=Constant3Vector() {constant=<rgba 0 0.2453 0.4219 0>}
    B=Constant3Vector() {constant=<rgba 0 0.5333 0.8308 0>}
    C=Constant3Vector() {constant=<rgba 0 1 0.7457 0>}
    Alpha=Multiply(A=@n44.R, B=@n45.R)
   ) {materialFunction=&Lerp_3Color}.Result
   B=Multiply(A=Power(Base=@n52.R, Exp=@n53), B=ScalarParameter() {defaultValue=1, group=Caustic, parameterName=Caustic_Intensity, sortPriority=3})
  ) {constB=0.05}
  n55 = TextureCoordinate()
  n56 = ComponentMask(None=@n55) {g=true}
  n57 = ScalarParameter() {defaultValue=1, group=Specular, parameterName=S_Whirl_Power, sortPriority=1}
  n58 = AppendVector(
   A=Multiply(
    A=Power(Base=@n56, Exp=@n57) {constExponent=1}
    B=ScalarParameter() {defaultValue=1, group=Specular, parameterName=S_Whirl_Intensity, sortPriority=2}
   )
   B=Constant()
  )
  n59 = VectorParameter() {group=Specular, parameterName=Specular_UV_Ctrl, defaultValue=<rgba 5 2 0.2 -0.5>}
  n60 = Multiply(A=@n55, B=AppendVector(A=@n59.R, B=@n59.R))
  n61 = Rotator(Coordinate=Panner(Coordinate=Add(A=@n58, B=@n60), Speed=AppendVector(A=@n59.B, B=@n59.B)), Time=Constant() {r=4.7125}) {speed=1}
  n62 = TextureSample(UVs=@n61) {
   samplerSource=SSM_Wrap_WorldGroupSettings
   samplerType=SAMPLERTYPE_Normal
   texture=/Game/Packages/Fortnite_UniqueMaterials_1/Textures/T_WaterFall_Tile_N
  }
  n63 = StaticBool()
  n64 = TextureSample!(UVs=Add(A=@n61, B=Constant2Vector() {r=-0.0025}), Tex, Bias=Constant() {r=1}, Apply View MipBias=@n63) {
   mipValueMode=TMVM_MipBias
   samplerSource=SSM_Wrap_WorldGroupSettings
   samplerType=SAMPLERTYPE_Normal
   texture=&T_WaterFall_Tile_N
  }
  n65 = TextureSample(UVs=Add(A=@n61, B=Constant2Vector() {r=-0.005}), Bias=Constant() {r=2}, Apply View MipBias=@n63) {
   mipValueMode=TMVM_MipBias
   samplerSource=SSM_Wrap_WorldGroupSettings
   samplerType=SAMPLERTYPE_Normal
   texture=&T_WaterFall_Tile_N
  }
  n66 = TextureSample(UVs=Add(A=@n61, B=Constant2Vector() {r=-0.0075}), Bias=Constant() {r=2}, Apply View MipBias=@n63) {
   mipValueMode=TMVM_MipBias
   samplerSource=SSM_Wrap_WorldGroupSettings
   samplerType=SAMPLERTYPE_Normal
   texture=&T_WaterFall_Tile_N
  }
  n67 = Add(A=Add(A=Add(A=@n62.RGB, B=@n64.RGB), B=@n65.RGB), B=@n66.RGB)
  n68 = QualitySwitch(Default=ShadingPathSwitch(Default=Divide(A=@n67) {constB=4}, Mobile=@n62.RGB) {desc=Migrated from FeatureLevelSwitch}, Low=@n62.RGB)
  n69 = TextureSample!(UVs=@n61, Tex, Apply View MipBias) {samplerSource=SSM_Wrap_WorldGroupSettings, samplerType=SAMPLERTYPE_Normal, texture=/Game/Global/Textures/Default/Tileable_Patterns/1Normal}
  n70 = MaterialFunctionCall(Normal=@n69.RGB, Flatness=ScalarParameter() {defaultValue=0.925, group=Cubemap, parameterName=Cubemap Distortion Flatness}) {materialFunction=&FlattenNormal}
  n71 = QualitySwitch(
   Default=ShadingPathSwitch(
    Default=MaterialFunctionCall(BaseNormal=@n68, AdditionalNormal=@n70.Result) {materialFunction=&BlendAngleCorrectedNormals}.Result
    Mobile=@n68
   ) {desc=Migrated from FeatureLevelSwitch}
   Low=@n68
  )
  n72 = TextureSampleParameterCube!(UVs=Transform(None=@n71), Apply View MipBias) {
   group=Cubemap
   parameterName=Cubemap Texture
   samplerSource=SSM_Wrap_WorldGroupSettings
   texture=/Game/Packages/Fortress_Sky/SkyDome/MaterialInstances/SkyDome_Apollo/Cubemap_Apollo_Night
  }
  n73 = Multiply(
   A=Desaturation(None=@n72.RGB, Fraction=ScalarParameter() {defaultValue=1, group=Cubemap, parameterName=Cubemap Desaturation})
   B=VectorParameter() {group=Cubemap, parameterName=Cubemap Multiplier, defaultValue=<rgba 1 1 1 1>}.RGB
  )
  n74 = QualitySwitch(Default=ShadingPathSwitch(Default=Add(A=@n54, B=@n73), Mobile=@n54) {desc=Migrated from FeatureLevelSwitch}, Low=@n54)
  n75 = TextureCoordinate()
  n76 = ScalarParameter() {defaultValue=2, group=UVs, parameterName=UV SpeedLines U Mult}
  n77 = ScalarParameter() {defaultValue=0.05, group=UVs, parameterName=UV SpeedLines V Mult}
  n78 = ScalarParameter() {defaultValue=0.4, group=UVs, parameterName=UV SpeedLines U Pan Speed}
  n79 = ScalarParameter() {defaultValue=-0.25, group=UVs, parameterName=UV SpeedLines V Pan Speed}
  n80 = Constant() {r=4.7125}
  n81 = TextureSample(UVs=Rotator(
    Coordinate=MaterialFunctionCall!(UV Channel=@n75, U Freq Multiplier=@n76, V Freq Multiplier=@n77, Pan Speed U=@n78, Pan Speed V=@n79) {materialFunction=/Game/Effects/Niagara/MasterMaterials/Functions/UVs/MF_FX_TextureUVPanner}.Result
    Time=@n80
   ) {speed=1}) {samplerSource=SSM_Wrap_WorldGroupSettings, texture=/Game/Global/Textures/Default/Noise/Noise}
  n82 = ScalarParameter() {defaultValue=12, parameterName=Speed Lines Power}
  n83 = TextureCoordinate()
  n84 = ScalarParameter() {defaultValue=2, group=UVs, parameterName=UV Mist U Mult}
  n85 = ScalarParameter() {defaultValue=3, group=UVs, parameterName=UV Mist V Mult}
  n86 = ScalarParameter() {defaultValue=0.4, group=UVs, parameterName=UV Mist U Pan Speed}
  n87 = ScalarParameter() {defaultValue=-1, group=UVs, parameterName=UV Mist V Pan Speed}
  n88 = TextureSample(UVs=Rotator(
    Coordinate=MaterialFunctionCall(UV Channel=@n83, U Freq Multiplier=@n84, V Freq Multiplier=@n85, Pan Speed U=@n86, Pan Speed V=@n87) {materialFunction=&MF_FX_TextureUVPanner}.Result
    Time=@n80
   ) {speed=1}) {samplerSource=SSM_Wrap_WorldGroupSettings, samplerType=SAMPLERTYPE_Grayscale, texture=&LowResBlurredNoise}
  n89 = Add(
   A=Multiply(A=Power(Base=@n81.G, Exp=@n82) {constExponent=18}, B=VectorParameter() {parameterName=Speed Lines Color, defaultValue=<rgba 1 1 1 0>}.RGB) {constB=0.25}
   B=Multiply(A=Power(Base=@n88.R) {constExponent=4}, B=VectorParameter() {parameterName=Mist Color, defaultValue=<rgba 1 1 1 1>}.RGB)
  )
  n90 = DotProduct(
   A=Transform(None=VectorParameter() {parameterName=FakeLightPos, defaultValue=<rgba 0 0 0.1 1>}.RGB) {transformSourceType=TRANSFORMSOURCE_World, transformType=TRANSFORM_Tangent}
   B=@n71
  )
  n91 = Multiply(
   A=LinearInterpolate(
    A=VectorParameter() {parameterName=ColorBright, defaultValue=<rgba 1 1 1 1>}.RGB
    B=VectorParameter() {parameterName=ColorDark, defaultValue=<rgba 0 0 0 1>}.RGB
    Alpha=Clamp(None=@n90)
   )
   B=Add(A=Multiply(A=@n36.G, B=@n37), B=Add(A=@n74, B=@n89))
  ) {constA=1}
  n92 = LinearInterpolate(
   A=LinearInterpolate(A=VectorParameter() {parameterName=Tunnel Mist Color, defaultValue=<rgba 0.0137 0.0252 0.0252 0>}.RGB, Alpha=@n19)
   Alpha=@n21.WhiteWash
  )
  n93 = OneMinus(None=@n18.VGradient)
  n94 = ScalarParameter() {defaultValue=0.2, parameterName=GradientPowerOuterShadow}
  n95 = TextureCoordinate()
  n96 = Power(
   Base=ComponentMask(None=@n95) {g=true}
   Exp=ScalarParameter() {defaultValue=1, group=Foam, parameterName=FoamMask_Whirl_Power, sortPriority=1}
  ) {constExponent=1}
  n97 = ScalarParameter() {defaultValue=2, group=Foam, parameterName=FoamMask_Whirl_Intensity, sortPriority=2}
  n98 = VectorParameter() {group=Foam, parameterName=FoamMask_UV_Ctrl, defaultValue=<rgba 2 1 0 -0.6>}
  n99 = Add(A=AppendVector(A=Multiply(A=@n96, B=@n97), B=Constant()), B=Multiply(A=@n95, B=AppendVector(A=@n98.R, B=@n98.R)))
  n100 = AppendVector(A=@n98.B, B=@n98.B)
  n101 = TextureSampleParameter2D(UVs=Add(A=@n14, B=Panner(Coordinate=@n99, Speed=@n100)) {constB=0}) {group=Foam, parameterName=FoamMask, samplerType=SAMPLERTYPE_Grayscale, texture=/Game/Global/Textures/Default/Noise/TilingNoise06}
  n102 = LinearInterpolate(Alpha=@n101.R) {constA=-0.9}
  n103 = Multiply(A=Add(A=@n16, B=Multiply(A=@n102) {constB=0.3}) {constB=0}) {constB=0.67}
  n104 = OneMinus(None=ComponentMask(None=TextureCoordinate()) {g=true})
  n105 = Add(A=Multiply(A=@n103, B=@n104), B=@n104)
  n106 = DynamicParameter() {defaultValue=<rgba 0 1 0 0>, paramNames [WhiteWash, Final Emissive Multiplier, Foam_Threshold, FoamFrame_Offset]}
  n107 = Subtract(A=Multiply(A=@n106.WhiteWash, B=Add(A=@n106.Foam_Threshold)) {constB=1.2}, B=@n106.Foam_Threshold) {constB=0.2}
  n108 = CurveAtlasRowParameter(CurveTime=Multiply(A=Add(A=@n105, B=@n107), B=ScalarParameter() {defaultValue=1, parameterName=FoamIntensity})) {
   atlas=&AthenaCurveAtlas01
   curve=/Game/Athena/Items/EnvironmentalItems/HidingProps/Props/WilliePete/Curve_WP
   defaultValue=218
  }
  n109 = Add(
   A=Multiply(A=Multiply(A=@n91, B=@n92), B=Power(Base=@n93, Exp=@n94))
   B=MaterialFunctionCall(
    A=VectorParameter() {parameterName=FoamColor_Low, defaultValue=<rgba 0 0 0 1>}.RGB
    B=VectorParameter() {parameterName=FoamColor_Mid, defaultValue=<rgba 0.4067 0.7884 0.8308 1>}.RGB
    C=VectorParameter() {parameterName=FoamColor_Hi, defaultValue=<rgba 0.4167 0.8347 1 1>}.RGB
    Alpha=@n108.R
   ) {materialFunction=&Lerp_3Color}.Result
  ) {constB=0}
  n110 = ParticleColor()
  MaterialFunctionCall(Emissive=Multiply(A=Multiply(A=@n109, B=@n21.Final Emissive Multiplier), B=@n110.R)) {materialFunction=&MF_FX_EmissiveAdjustments}
  n111 = DynamicParameter() {defaultValue=<rgba 0 1 0 0>, paramNames [WhiteWash, Final Emissive Multiplier, Foam_Threshold, FoamFrame_Offset]}
  n112 = Add(A=ComponentMask(None=TextureCoordinate()) {g=true}, B=@n111.FoamFrame_Offset)
  n113 = CurveAtlasRowParameter(CurveTime=@n112) {
   atlas=&AthenaCurveAtlas01
   curve=&Curve_WP
   defaultValue=218
  }
  n114 = Multiply(A=@n113.B, B=@n103)
  n115 = CurveAtlasRowParameter(CurveTime=Multiply(A=Add(A=@n112, B=@n114)) {constB=0.5}) {
   atlas=&AthenaCurveAtlas01
   curve=&Curve_WP
   defaultValue=218
  }
  Multiply(A=@n110.A, B=LinearInterpolate(B=@n115.G, Alpha=ScalarParameter() {parameterName=FrameFoam}) {constA=1})
 ]
}
